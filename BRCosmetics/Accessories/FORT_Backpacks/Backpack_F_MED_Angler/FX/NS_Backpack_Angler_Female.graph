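# Niagara effect for the female Angler backpack cosmetic, adding a pulsing blue glow sprite and small light.
# A single emitter bursts one persistent particle at an offset from the system, feeding both light and sprite renderers.
# Normalized loop age drives curves that pulse light color and sprite alpha.
# User.StatLerpKill blends the sprite toward a brighter, larger look with non-pulsing alpha, while the other user parameters are unused in the visible modules.
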NiagaraSystem /BRCosmetics/Accessories/FORT_Backpacks/Backpack_F_MED_Angler/FX/NS_Backpack_Angler_Female
dependencies {
 usedDataInterfaces [&NiagaraDataInterfaceColorCurve, &NiagaraDataInterfaceCurve]
 usedDynamicInputs []
 usedModules [&SystemState]
}
emitters {Glow_and_Light {
  emitterUpdateScript [
   &EmitterState(Inactive Response=::NewEnumerator2, Life Cycle Mode=::NewEnumerator1, Loop Behavior=::NewEnumerator1, Loop Duration=2, UseLoopDelay=false)
   &SpawnBurst_Instantaneous(Spawn Count=1, Use Spawn Probability=false)
  ]
  particleSpawnScript [
   /Game/Effects/Niagara/Modules/Groups/FN_SetSpawnGroup_v1/FN_SetSpawnGroup(Emitter Bone Count=2)
   /Niagara/Modules/Spawn/Initialization/InitializeParticle(
    Color=<rgba 0.2359 0.8154 1.5 0.75>
    Lifetime=0
    Module.Write Color=true
    Module.Write Lifetime=true
    Module.Write Mass=false
    Module.Write Material Random=true
    Module.Write Position=false
    Module.Write Scale=false
    Module.Write SpriteRotation=false
    Module.Write SpriteSize=true
    Position=&SimulationPosition()
    Sprite Size=<42.5 42.5>
   )
   &SystemLocation(MaskSpawnGroup=false, Offset=<0 -10 -2>)
   /SetVariables(Particles.GlowOffsetLocation:Vector3f=&Add_Vector(A=Linked() {linkedVariable {name=Particles.Position, type=NiagaraPosition}}, B=<0 0 8>)) {isSetParameters=true}
   &CameraOffset(Camera Offset Amount=5)
  ]
  particleUpdateScript [
   &ParticleState(Kill Particles When Lifetime Has Elapsed=false)
   &SolveForcesAndVelocity(Clamp Velocity=false, Limit Acceleration=false)
   /Game/Curie/Effects/LightAttributeSetter(
    LightColor:LinearColor=&ColorFromCurve(
     ColorCurve=DataInterface() {propertyValues {
       AlphaCurve {keys [<at 0 1 RCIM_Cubic>,<at 1 1 RCIM_Cubic>]}
       BlueCurve {keys [<at 0 2 RCIM_Cubic>,<at 0.5 0.5>,<at 1 2 RCIM_Cubic>]}
       GreenCurve {keys [<at 0 1.0873 RCIM_Cubic>,<at 0.5 0.2718>,<at 1 1.0873 RCIM_Cubic>]}
       RedCurve {keys [<at 0 0.3146 RCIM_Cubic>,<at 0.5 0.0786>,<at 1 0.3146 RCIM_Cubic>]}
      }}
     CurveIndex=$Emitter.NormalizedLoopAge
    )
    LightRadius=50
   )
   &ScaleColor(
    Scale Alpha=&Lerp_Float(
     A=&FloatFromCurve(
      FloatCurve=DataInterface() {propertyValues {Curve {keys [<at 0 0.35 RCIM_Cubic>,<at 0.5 1 RCIM_Cubic>,<at 1 0.35 RCIM_Cubic>]}}}
      CurveIndex=$Emitter.NormalizedLoopAge
     )
     B=&FloatFromCurve(
      FloatCurve=DataInterface() {propertyValues {Curve {keys [<at 0 0.5>,<at 1 0.5>]}}}
      CurveIndex=$Emitter.NormalizedLoopAge
      Scale Curve=$User.StatLerpKill
     )
     Alpha=$User.StatLerpKill
    )
    Scale RGB=&Lerp_Vector(
     A=&MakeVectorFromLinearColorRGB(Color=<rgba 1 1 1 1>)
     B=&MakeVectorFromLinearColorRGB(Color=<rgba 1.8125 2.1277 2.5 1>)
     Alpha=$User.StatLerpKill
    )
    ScaleA=true
    ScaleRGB=true
   )
   &ScaleSpriteSize(Scale Factor=&Vector2DFromFloat(Value=&Lerp_Float(A=1, B=1.25, Alpha=$User.StatLerpKill)))
  ]
  rendererClasses [&NiagaraLightRendererProperties, &NiagaraSpriteRendererProperties]
  renderers [
   {rendererClass=&NiagaraLightRendererProperties, rendererIndex=0}
   {rendererClass=&NiagaraSpriteRendererProperties, rendererIndex=1}
  ]
 }}
summary {
 systemName=NS_Backpack_Angler_Female
 emitters [{
   emitterName=Glow_and_Light
   rendererClasses [&NiagaraLightRendererProperties, &NiagaraSpriteRendererProperties]
  }]
 userVariables [
  User.bIsFrontEndPreview {type=NiagaraBool}
  User.IsJumping {type=NiagaraBool}
  User.IsSprinting {type=NiagaraBool}
  User.PlayerSpeed {type=NiagaraFloat, defaultValue=0, description=""}
  User.SkeletalMesh {
   type=NiagaraDataInterfaceSkeletalMesh
   defaultValue {
    struct=NiagaraExt_VariableValue_DataInterface
    value {dataInterface=&self:NiagaraDataInterfaceSkeletalMesh_0}
   }
  }
  User.SlurpColorAlpha {type=NiagaraFloat, defaultValue=0, description=""}
  User.StatLerpKill {type=NiagaraFloat, defaultValue=0, description=""}
 ]
}
