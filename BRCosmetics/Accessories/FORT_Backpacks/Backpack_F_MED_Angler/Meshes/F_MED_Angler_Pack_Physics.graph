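# Physics asset for the Angler backpack cosmetic, defining collision bodies and joints for its pack, hood, stem and tentacles.
# A pack capsule anchors the hood and a small body capsule, which carries a three-sphere stem.
# Six five-sphere tentacle chains branch from the stem's last bone.
# Swing limits loosen along the stem and are wider at tentacle roots than along the chains, while twist stays uniformly tight.
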PhysicsAsset /BRCosmetics/Accessories/FORT_Backpacks/Backpack_F_MED_Angler/Meshes/F_MED_Angler_Pack_Physics
bodies {
 body [{shapeType=Capsule, length=1, radius=1, center=<0.1 0 0>, rotation=<rot 90 0 0>}]
 hood [{shapeType=Sphere, radius=3.1143, center=<0.6416 0.0043 0.0318>}]
 pack [{shapeType=Capsule, length=11.5, radius=6.25, shapeName=pack_capsule, center=<0 7.6476 -0.7214>, rotation=<rot -2.522 -90 0>}]
 stem_01 [{shapeType=Sphere, radius=2.25, shapeName=stem_01_sphere, center=<0.6616 0.0436 0>}]
 stem_02 [{shapeType=Sphere, radius=2.25, shapeName=stem_02_sphere, center=<0.7262 0 0>}]
 stem_03 [{shapeType=Sphere, radius=2.2943, shapeName=stem_03_sphere, center=<0.5608 -0.1109 0.1092>}]
 tentacle_bck_01 [{shapeType=Sphere, radius=1.5, shapeName=tentacle_bck_01_sphere, center=<0.2722 -1.038 0.0316>}]
 tentacle_bck_02 [{shapeType=Sphere, radius=1.5298, shapeName=tentacle_bck_02_sphere, center=<-0.1146 -0.0408 -0.0344>}]
 tentacle_bck_03 [{shapeType=Sphere, radius=1.5, shapeName=tentacle_bck_03_sphere, center=<-0.2592 0.04 -0.0837>}]
 tentacle_bck_04 [{shapeType=Sphere, radius=1.5, shapeName=tentacle_bck_04_sphere, center=<0.3717 0.0509 -0.1297>}]
 tentacle_bck_05 [{shapeType=Sphere, radius=1.5, shapeName=tentacle_bck_05_sphere, center=<0.5699 0.0603 -0.1953>}]
 tentacle_bck_l_01 [{shapeType=Sphere, radius=1.506, shapeName=tentacle_bck_l_01_sphere, center=<0.2489 -0.9215 0>}]
 tentacle_bck_l_02 [{shapeType=Sphere, radius=1.5, shapeName=tentacle_bck_l_02_sphere, center=<0.0314 -0.0522 0>}]
 tentacle_bck_l_03 [{shapeType=Sphere, radius=1.5, shapeName=tentacle_bck_l_03_sphere, center=<0.3777 -0.0169 0>}]
 tentacle_bck_l_04 [{shapeType=Sphere, radius=1.5, shapeName=tentacle_bck_l_04_sphere, center=<0.3512 -0.0172 0>}]
 tentacle_bck_l_05 [{shapeType=Sphere, radius=1.5, shapeName=tentacle_bck_l_05_sphere, center=<0.5518 -0.0418 0>}]
 tentacle_bck_r_01 [{shapeType=Sphere, radius=1.5, shapeName=tentacle_bck_r_01_sphere, center=<0.2489 -1.211 -0.3112>}]
 tentacle_bck_r_02 [{shapeType=Sphere, radius=1.5, shapeName=tentacle_bck_r_02_sphere, center=<0.0687 -0.336 -0.2245>}]
 tentacle_bck_r_03 [{shapeType=Sphere, radius=1.5, shapeName=tentacle_bck_r_03_sphere, center=<-0.0334 -0.3065 0.0226>}]
 tentacle_bck_r_04 [{shapeType=Sphere, radius=1.5, shapeName=tentacle_bck_r_04_sphere, center=<0.3512 -0.3068 0.0244>}]
 tentacle_bck_r_05 [{shapeType=Sphere, radius=1.5, shapeName=tentacle_bck_r_05_sphere, center=<0.5518 -0.3314 0.0256>}]
 tentacle_frt_01 [{shapeType=Sphere, radius=1.5, shapeName=tentacle_frt_01_sphere, center=<0.023 -0.4998 0.0846>}]
 tentacle_frt_02 [{shapeType=Sphere, radius=1.5, shapeName=tentacle_frt_02_sphere, center=<-0.4293 -0.219 0.0091>}]
 tentacle_frt_03 [{shapeType=Sphere, radius=1.5, shapeName=tentacle_frt_03_sphere, center=<0.086 -0.1344 0.0377>}]
 tentacle_frt_04 [{shapeType=Sphere, radius=1.5, shapeName=tentacle_frt_04_sphere, center=<0.4347 -0.1711 0.0308>}]
 tentacle_frt_05 [{shapeType=Sphere, radius=1.5, shapeName=tentacle_frt_05_sphere, center=<0.2514 -0.1183 0.0134>}]
 tentacle_mid_l_01 [{shapeType=Sphere, radius=1.5075, shapeName=tentacle_mid_l_01_sphere, center=<0.2379 -0.8919 0.2494>}]
 tentacle_mid_l_02 [{shapeType=Sphere, radius=1.5075, shapeName=tentacle_mid_l_02_sphere, center=<-0.4691 -0.4931 0.1594>}]
 tentacle_mid_l_03 [{shapeType=Sphere, radius=1.5, shapeName=tentacle_mid_l_03_sphere, center=<-0.1562 -0.1154 -0.0908>}]
 tentacle_mid_l_04 [{shapeType=Sphere, radius=1.5, shapeName=tentacle_mid_l_04_sphere, center=<0.2191 -0.03 -0.0918>}]
 tentacle_mid_l_05 [{shapeType=Sphere, radius=1.5, shapeName=tentacle_mid_l_05_sphere, center=<0.2409 0.0096 -0.0816>}]
 tentacle_mid_r_01 [{shapeType=Sphere, radius=1.5075, shapeName=tentacle_mid_r_01_sphere, center=<0.2379 -1.1814 -0.2495>}]
 tentacle_mid_r_02 [{shapeType=Sphere, radius=1.5075, shapeName=tentacle_mid_r_02_sphere, center=<-0.1651 -0.3427 -0.1594>}]
 tentacle_mid_r_03 [{shapeType=Sphere, radius=1.5, shapeName=tentacle_mid_r_03_sphere, center=<-0.2942 -0.273 0.0908>}]
 tentacle_mid_r_04 [{shapeType=Sphere, radius=1.5, shapeName=tentacle_mid_r_04_sphere, center=<0.4235 -0.3195 0.0918>}]
 tentacle_mid_r_05 [{shapeType=Sphere, radius=1.5, shapeName=tentacle_mid_r_05_sphere, center=<0.2409 -0.28 0.0817>}]
}
constraints [
 {bone1Name=body, bone2Name=pack, swing1LimitDegrees=50, swing2LimitDegrees=50, twistLimitDegrees=10}
 {bone1Name=stem_01, bone2Name=body, swing1LimitDegrees=10, swing2LimitDegrees=10, twistLimitDegrees=10}
 {bone1Name=stem_02, bone2Name=stem_01, swing1LimitDegrees=20, swing2LimitDegrees=20, twistLimitDegrees=10}
 {bone1Name=stem_03, bone2Name=stem_02, swing1LimitDegrees=30, swing2LimitDegrees=30, twistLimitDegrees=10}
 {bone1Name=tentacle_bck_l_01, bone2Name=stem_03, swing1LimitDegrees=50, swing2LimitDegrees=50, twistLimitDegrees=10}
 {bone1Name=tentacle_bck_l_02, bone2Name=tentacle_bck_l_01, swing1LimitDegrees=30, swing2LimitDegrees=30, twistLimitDegrees=10}
 {bone1Name=tentacle_bck_l_03, bone2Name=tentacle_bck_l_02, swing1LimitDegrees=30, swing2LimitDegrees=30, twistLimitDegrees=10}
 {bone1Name=tentacle_bck_l_04, bone2Name=tentacle_bck_l_03, swing1LimitDegrees=30, swing2LimitDegrees=30, twistLimitDegrees=10}
 {bone1Name=tentacle_bck_l_05, bone2Name=tentacle_bck_l_04, swing1LimitDegrees=30, swing2LimitDegrees=30, twistLimitDegrees=10}
 {bone1Name=tentacle_mid_l_01, bone2Name=stem_03, swing1LimitDegrees=50, swing2LimitDegrees=50, twistLimitDegrees=10}
 {bone1Name=tentacle_mid_l_02, bone2Name=tentacle_mid_l_01, swing1LimitDegrees=30, swing2LimitDegrees=30, twistLimitDegrees=10}
 {bone1Name=tentacle_mid_l_03, bone2Name=tentacle_mid_l_02, swing1LimitDegrees=30, swing2LimitDegrees=30, twistLimitDegrees=10}
 {bone1Name=tentacle_mid_l_04, bone2Name=tentacle_mid_l_03, swing1LimitDegrees=30, swing2LimitDegrees=30, twistLimitDegrees=10}
 {bone1Name=tentacle_mid_l_05, bone2Name=tentacle_mid_l_04, swing1LimitDegrees=30, swing2LimitDegrees=30, twistLimitDegrees=10}
 {bone1Name=tentacle_frt_01, bone2Name=stem_03, swing1LimitDegrees=50, swing2LimitDegrees=50, twistLimitDegrees=10}
 {bone1Name=tentacle_frt_02, bone2Name=tentacle_frt_01, swing1LimitDegrees=30, swing2LimitDegrees=30, twistLimitDegrees=10}
 {bone1Name=tentacle_frt_03, bone2Name=tentacle_frt_02, swing1LimitDegrees=30, swing2LimitDegrees=30, twistLimitDegrees=10}
 {bone1Name=tentacle_frt_04, bone2Name=tentacle_frt_03, swing1LimitDegrees=30, swing2LimitDegrees=30, twistLimitDegrees=10}
 {bone1Name=tentacle_frt_05, bone2Name=tentacle_frt_04, swing1LimitDegrees=30, swing2LimitDegrees=30, twistLimitDegrees=10}
 {bone1Name=tentacle_bck_01, bone2Name=stem_03, swing1LimitDegrees=50, swing2LimitDegrees=50, twistLimitDegrees=10}
 {bone1Name=tentacle_bck_02, bone2Name=tentacle_bck_01, swing1LimitDegrees=30, swing2LimitDegrees=30, twistLimitDegrees=10}
 {bone1Name=tentacle_bck_03, bone2Name=tentacle_bck_02, swing1LimitDegrees=30, swing2LimitDegrees=30, twistLimitDegrees=10}
 {bone1Name=tentacle_bck_04, bone2Name=tentacle_bck_03, swing1LimitDegrees=30, swing2LimitDegrees=30, twistLimitDegrees=10}
 {bone1Name=tentacle_bck_05, bone2Name=tentacle_bck_04, swing1LimitDegrees=30, swing2LimitDegrees=30, twistLimitDegrees=10}
 {bone1Name=tentacle_mid_r_01, bone2Name=stem_03, swing1LimitDegrees=50, swing2LimitDegrees=50, twistLimitDegrees=10}
 {bone1Name=tentacle_mid_r_02, bone2Name=tentacle_mid_r_01, swing1LimitDegrees=30, swing2LimitDegrees=30, twistLimitDegrees=10}
 {bone1Name=tentacle_mid_r_03, bone2Name=tentacle_mid_r_02, swing1LimitDegrees=30, swing2LimitDegrees=30, twistLimitDegrees=10}
 {bone1Name=tentacle_mid_r_04, bone2Name=tentacle_mid_r_03, swing1LimitDegrees=30, swing2LimitDegrees=30, twistLimitDegrees=10}
 {bone1Name=tentacle_mid_r_05, bone2Name=tentacle_mid_r_04, swing1LimitDegrees=30, swing2LimitDegrees=30, twistLimitDegrees=10}
 {bone1Name=tentacle_bck_r_01, bone2Name=stem_03, swing1LimitDegrees=50, swing2LimitDegrees=50, twistLimitDegrees=10}
 {bone1Name=tentacle_bck_r_02, bone2Name=tentacle_bck_r_01, swing1LimitDegrees=30, swing2LimitDegrees=30, twistLimitDegrees=10}
 {bone1Name=tentacle_bck_r_03, bone2Name=tentacle_bck_r_02, swing1LimitDegrees=30, swing2LimitDegrees=30, twistLimitDegrees=10}
 {bone1Name=tentacle_bck_r_04, bone2Name=tentacle_bck_r_03, swing1LimitDegrees=30, swing2LimitDegrees=30, twistLimitDegrees=10}
 {bone1Name=tentacle_bck_r_05, bone2Name=tentacle_bck_r_04, swing1LimitDegrees=30, swing2LimitDegrees=30, twistLimitDegrees=10}
 {bone1Name=hood, bone2Name=pack, swing1LimitDegrees=50, swing2LimitDegrees=50, twistLimitDegrees=10}
]
