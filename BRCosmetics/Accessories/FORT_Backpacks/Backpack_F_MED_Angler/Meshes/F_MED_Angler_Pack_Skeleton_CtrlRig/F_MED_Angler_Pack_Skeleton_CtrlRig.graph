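# Control Rig for the Angler backpack cosmetic that procedurally animates its six tentacles and hood.
# Seven ChainHarmonics units run in sequence, one per root bone, layering pendulum and wave motion with debug drawing enabled.
# The pawnSpeed variable drives each unit's Z speed, while slightly varied wave offsets differentiate some chains.
# The hood, a single childless bone, also gets wave amplitude on all three axes.
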ControlRigBlueprint /BRCosmetics/Accessories/FORT_Backpacks/Backpack_F_MED_Angler/Meshes/F_MED_Angler_Pack_Skeleton_CtrlRig/F_MED_Angler_Pack_Skeleton_CtrlRig
graphs [
 RigVMFunctionLibrary {links [], nodes []}
 RigVMModel {nodes [
   RigUnit_BeginExecution = &RigUnit_BeginExecution()
   n1 = RigVMVariableNode() {variable=pawnSpeed}
   /Script/ControlRig.RigUnit_ChainHarmonics(
    ChainRoot=tentacle_frt_01
    DrawWorldOffset=(Rotation=(X=0,Y=0,Z=0,W=1),Translation=(X=0,Y=0,Z=0),Scale3D=(X=1,Y=1,Z=1))
    Pendulum=(bEnabled=True,PendulumStiffness=2,PendulumGravity=(X=0,Y=0,Z=0),PendulumBlend=0.75,PendulumDrag=0.98,PendulumMaximum=0.1,UnwindAxis=(X=0,Y=1,Z=0),UnwindMinimum=0.2,UnwindMaximum=0.05)
    Reach=(bEnabled=True,ReachTarget=(X=0,Y=0,Z=0),ReachAxis=(X=1,Y=0,Z=0),ReachMinimum=0,ReachMaximum=0)
    Speed=(X=1,Y=1,Z=1)
    Wave=(bEnabled=True,WaveFrequency=(X=1,Y=0.6,Z=-3),WaveAmplitude=(X=0,Y=0,Z=1.5),WaveOffset=(X=0,Y=0,Z=2),WaveNoise=(X=0,Y=0,Z=0),WaveMinimum=0,WaveMaximum=1.5)
    WaveCurve=(EditorCurveData=(Keys=((Value=0.5),(Time=0.9884,Value=0.75)),DefaultValue=340282346638528859811704183484516925440,PreInfinityExtrap=RCCE_Constant,PostInfinityExtrap=RCCE_Constant))
    bDrawDebug=True
    ExecutePin=@RigUnit_BeginExecution.ExecutePin
    Speed.Z=@n1.Value
   ) {is=struct RigVMUnitNode}
   > &RigUnit_ChainHarmonics(
    ChainRoot=tentacle_bck_01
    DrawWorldOffset=(Rotation=(X=0,Y=0,Z=0,W=1),Translation=(X=0,Y=0,Z=0),Scale3D=(X=1,Y=1,Z=1))
    Pendulum=(bEnabled=True,PendulumStiffness=2,PendulumGravity=(X=0,Y=0,Z=0),PendulumBlend=0.75,PendulumDrag=0.98,PendulumMaximum=0.1,UnwindAxis=(X=0,Y=1,Z=0),UnwindMinimum=0.2,UnwindMaximum=0.05)
    Reach=(bEnabled=True,ReachTarget=(X=0,Y=0,Z=0),ReachAxis=(X=1,Y=0,Z=0),ReachMinimum=0,ReachMaximum=0)
    Speed=(X=1,Y=1,Z=1)
    Wave=(bEnabled=True,WaveFrequency=(X=1,Y=0.6,Z=-3),WaveAmplitude=(X=0,Y=0,Z=1.5),WaveOffset=(X=0,Y=0,Z=2),WaveNoise=(X=0,Y=0,Z=0),WaveMinimum=0,WaveMaximum=1.5)
    WaveCurve=(EditorCurveData=(Keys=((Value=0.5),(Time=1,Value=0.75)),DefaultValue=340282346638528859811704183484516925440,PreInfinityExtrap=RCCE_Constant,PostInfinityExtrap=RCCE_Constant))
    bDrawDebug=True
    Speed.Z=@n1.Value
   )
   > &RigUnit_ChainHarmonics(
    ChainRoot=tentacle_bck_l_01
    DrawWorldOffset=(Rotation=(X=0,Y=0,Z=0,W=1),Translation=(X=0,Y=0,Z=0),Scale3D=(X=1,Y=1,Z=1))
    Pendulum=(bEnabled=True,PendulumStiffness=2,PendulumGravity=(X=0,Y=0,Z=0),PendulumBlend=0.75,PendulumDrag=0.98,PendulumMaximum=0.1,UnwindAxis=(X=0,Y=1,Z=0),UnwindMinimum=0.2,UnwindMaximum=0.05)
    Reach=(bEnabled=True,ReachTarget=(X=0,Y=0,Z=0),ReachAxis=(X=1,Y=0,Z=0),ReachMinimum=0,ReachMaximum=0)
    Speed=(X=1,Y=1,Z=1)
    Wave=(bEnabled=True,WaveFrequency=(X=1,Y=0.6,Z=-3),WaveAmplitude=(X=0,Y=0,Z=1.5),WaveOffset=(X=0,Y=0,Z=1.9),WaveNoise=(X=0,Y=0,Z=0),WaveMinimum=0,WaveMaximum=1.5)
    WaveCurve=(EditorCurveData=(Keys=((Value=0.5),(Time=1,Value=0.75)),DefaultValue=340282346638528859811704183484516925440,PreInfinityExtrap=RCCE_Constant,PostInfinityExtrap=RCCE_Constant))
    bDrawDebug=True
    Speed.Z=@n1.Value
   )
   > &RigUnit_ChainHarmonics(
    ChainRoot=tentacle_mid_l_01
    DrawWorldOffset=(Rotation=(X=0,Y=0,Z=0,W=1),Translation=(X=0,Y=0,Z=0),Scale3D=(X=1,Y=1,Z=1))
    Pendulum=(bEnabled=True,PendulumStiffness=2,PendulumGravity=(X=0,Y=0,Z=0),PendulumBlend=0.75,PendulumDrag=0.98,PendulumMaximum=0.1,UnwindAxis=(X=0,Y=1,Z=0),UnwindMinimum=0.2,UnwindMaximum=0.05)
    Reach=(bEnabled=True,ReachTarget=(X=0,Y=0,Z=0),ReachAxis=(X=1,Y=0,Z=0),ReachMinimum=0,ReachMaximum=0)
    Speed=(X=1,Y=1,Z=1)
    Wave=(bEnabled=True,WaveFrequency=(X=1,Y=0.6,Z=-3),WaveAmplitude=(X=0,Y=0,Z=1.5),WaveOffset=(X=0,Y=0,Z=2),WaveNoise=(X=0,Y=0,Z=0),WaveMinimum=0,WaveMaximum=1.5)
    WaveCurve=(EditorCurveData=(Keys=((Value=0.5),(Time=1,Value=0.75)),DefaultValue=340282346638528859811704183484516925440,PreInfinityExtrap=RCCE_Constant,PostInfinityExtrap=RCCE_Constant))
    bDrawDebug=True
    Speed.Z=@n1.Value
   )
   > &RigUnit_ChainHarmonics(
    ChainRoot=tentacle_mid_r_01
    DrawWorldOffset=(Rotation=(X=0,Y=0,Z=0,W=1),Translation=(X=0,Y=0,Z=0),Scale3D=(X=1,Y=1,Z=1))
    Pendulum=(bEnabled=True,PendulumStiffness=2,PendulumGravity=(X=0,Y=0,Z=0),PendulumBlend=0.75,PendulumDrag=0.98,PendulumMaximum=0.1,UnwindAxis=(X=0,Y=1,Z=0),UnwindMinimum=0.2,UnwindMaximum=0.05)
    Reach=(bEnabled=True,ReachTarget=(X=0,Y=0,Z=0),ReachAxis=(X=1,Y=0,Z=0),ReachMinimum=0,ReachMaximum=0)
    Speed=(X=1,Y=1,Z=1)
    Wave=(bEnabled=True,WaveFrequency=(X=1,Y=0.6,Z=-3),WaveAmplitude=(X=0,Y=0,Z=1.5),WaveOffset=(X=0,Y=0,Z=2.1),WaveNoise=(X=0,Y=0,Z=0),WaveMinimum=0,WaveMaximum=1.5)
    WaveCurve=(EditorCurveData=(Keys=((Value=0.5),(Time=1,Value=0.75)),DefaultValue=340282346638528859811704183484516925440,PreInfinityExtrap=RCCE_Constant,PostInfinityExtrap=RCCE_Constant))
    bDrawDebug=True
    Speed.Z=@n1.Value
   )
   > &RigUnit_ChainHarmonics(
    ChainRoot=tentacle_bck_r_01
    DrawWorldOffset=(Rotation=(X=0,Y=0,Z=0,W=1),Translation=(X=0,Y=0,Z=0),Scale3D=(X=1,Y=1,Z=1))
    Pendulum=(bEnabled=True,PendulumStiffness=2,PendulumGravity=(X=0,Y=0,Z=0),PendulumBlend=0.75,PendulumDrag=0.98,PendulumMaximum=0.1,UnwindAxis=(X=0,Y=1,Z=0),UnwindMinimum=0.2,UnwindMaximum=0.05)
    Reach=(bEnabled=True,ReachTarget=(X=0,Y=0,Z=0),ReachAxis=(X=1,Y=0,Z=0),ReachMinimum=0,ReachMaximum=0)
    Speed=(X=1,Y=1,Z=1)
    Wave=(bEnabled=True,WaveFrequency=(X=1,Y=0.6,Z=-3),WaveAmplitude=(X=0,Y=0,Z=1.5),WaveOffset=(X=0,Y=0,Z=2.4),WaveNoise=(X=0,Y=0,Z=0),WaveMinimum=0,WaveMaximum=1.5)
    WaveCurve=(EditorCurveData=(Keys=((Value=0.5),(Time=1,Value=0.75)),DefaultValue=340282346638528859811704183484516925440,PreInfinityExtrap=RCCE_Constant,PostInfinityExtrap=RCCE_Constant))
    bDrawDebug=True
    Speed.Z=@n1.Value
   )
   > &RigUnit_ChainHarmonics(
    ChainRoot=hood
    DrawWorldOffset=(Rotation=(X=0,Y=0,Z=0,W=1),Translation=(X=0,Y=0,Z=0),Scale3D=(X=1,Y=1,Z=1))
    Pendulum=(bEnabled=True,PendulumStiffness=2,PendulumGravity=(X=0,Y=0,Z=0),PendulumBlend=0.75,PendulumDrag=0.98,PendulumMaximum=0.1,UnwindAxis=(X=0,Y=1,Z=0),UnwindMinimum=0.2,UnwindMaximum=0.05)
    Reach=(bEnabled=True,ReachTarget=(X=0,Y=0,Z=0),ReachAxis=(X=1,Y=0,Z=0),ReachMinimum=0,ReachMaximum=0)
    Speed=(X=1,Y=1,Z=1)
    Wave=(bEnabled=True,WaveFrequency=(X=1,Y=0.6,Z=-3),WaveAmplitude=(X=1,Y=1,Z=1.5),WaveOffset=(X=0,Y=0,Z=2.3),WaveNoise=(X=0,Y=0,Z=0),WaveMinimum=0,WaveMaximum=1.5)
    WaveCurve=(EditorCurveData=(Keys=((Value=0.5),(Time=1,Value=0.75)),DefaultValue=340282346638528859811704183484516925440,PreInfinityExtrap=RCCE_Constant,PostInfinityExtrap=RCCE_Constant))
    bDrawDebug=True
    Speed.Z=@n1.Value
   )
  ]}
]
hierarchy [
 root {loc=<0 0 0>, rot=<0 0 0 1>}
 pack {parent=root, loc=<0 -0.7316 0>, rot=<0 0 0 1>}
 body_MSTR {parent=pack, loc=<0 -9.3654 5.4313>, rot=<0 -0.7071 0 -0.7071>}
 body {parent=body_MSTR, loc=<0 0 0>, rot=<0 0 0 -1>}
 stem_01 {parent=body, loc=<3.6086 0 0>, rot=<0 0 0 -1>}
 stem_02 {parent=stem_01, loc=<3.6954 0 0>, rot=<0 0 0 -1>}
 stem_03 {parent=stem_02, loc=<3.2975 0 0>, rot=<0 0 0 -1>}
 tentacle_bck_l_01 {parent=stem_03, loc=<2.2097 2.4942 4.32>, rot=<0.8495 0.1823 0.1042 -0.484>}
 tentacle_bck_l_02 {parent=tentacle_bck_l_01, loc=<3.6784 0 0>, rot=<0 0 -0.0462 -0.9989>}
 tentacle_bck_l_03 {parent=tentacle_bck_l_02, loc=<2.6804 0 0>, rot=<0 0 -0.0906 -0.9959>}
 tentacle_bck_l_04 {parent=tentacle_bck_l_03, loc=<2.0857 0 0>, rot=<0 0 -0.0471 -0.9989>}
 tentacle_bck_l_05 {parent=tentacle_bck_l_04, loc=<2.4973 0 0>, rot=<0 0 0 -1>}
 tentacle_mid_l_01 {parent=stem_03, loc=<2.2097 -2.4942 4.32>, rot=<0.5161 0.108 0.1828 -0.8298>}
 tentacle_mid_l_02 {parent=tentacle_mid_l_01, loc=<3.6784 0 0>, rot=<0 0 -0.0462 -0.9989>}
 tentacle_mid_l_03 {parent=tentacle_mid_l_02, loc=<2.6804 0 0>, rot=<0 0 -0.0906 -0.9959>}
 tentacle_mid_l_04 {parent=tentacle_mid_l_03, loc=<2.0857 0 0>, rot=<0 0 -0.0471 -0.9989>}
 tentacle_mid_l_05 {parent=tentacle_mid_l_04, loc=<2.4973 0 0>, rot=<0 0 0 1>}
 tentacle_frt_01 {parent=stem_03, loc=<2.2097 -4.9883 0>, rot=<-0.0237 0.0005 0.2148 -0.9764>}
 tentacle_frt_02 {parent=tentacle_frt_01, loc=<3.6784 0 0>, rot=<0 0 -0.0462 -0.9989>}
 tentacle_frt_03 {parent=tentacle_frt_02, loc=<2.6804 0 0>, rot=<0 0 -0.0906 -0.9959>}
 tentacle_frt_04 {parent=tentacle_frt_03, loc=<2.0857 0 0>, rot=<0 0 -0.0471 -0.9989>}
 tentacle_frt_05 {parent=tentacle_frt_04, loc=<2.4973 0 0>, rot=<0 0 0 1>}
 tentacle_bck_01 {parent=stem_03, loc=<2.2097 4.988 0>, rot=<0.9764 0.2146 -0.0095 -0.0218>}
 tentacle_bck_02 {parent=tentacle_bck_01, loc=<3.6784 0 0>, rot=<0 0 -0.0462 -0.9989>}
 tentacle_bck_03 {parent=tentacle_bck_02, loc=<2.6804 0 0>, rot=<0 0 -0.0906 -0.9959>}
 tentacle_bck_04 {parent=tentacle_bck_03, loc=<2.0857 0 0>, rot=<0 0 -0.0471 -0.9989>}
 tentacle_bck_05 {parent=tentacle_bck_04, loc=<2.4973 0 0>, rot=<0 0 0 1>}
 tentacle_mid_r_01 {parent=stem_03, loc=<2.2097 -2.4942 -4.32>, rot=<-0.5187 -0.0948 0.1889 -0.8284>}
 tentacle_mid_r_02 {parent=tentacle_mid_r_01, loc=<3.6784 0 0>, rot=<0 0 -0.0462 -0.9989>}
 tentacle_mid_r_03 {parent=tentacle_mid_r_02, loc=<2.6804 0 0>, rot=<0 0 -0.0906 -0.9959>}
 tentacle_mid_r_04 {parent=tentacle_mid_r_03, loc=<2.0857 0 0>, rot=<0 0 -0.0471 -0.9989>}
 tentacle_mid_r_05 {parent=tentacle_mid_r_04, loc=<2.4973 0 0>, rot=<0 0 0 -1>}
 tentacle_bck_r_01 {parent=stem_03, loc=<2.2097 2.494 -4.32>, rot=<-0.8323 -0.1844 0.105 -0.5121>}
 tentacle_bck_r_02 {parent=tentacle_bck_r_01, loc=<3.6784 0 0>, rot=<0 0 -0.0462 -0.9989>}
 tentacle_bck_r_03 {parent=tentacle_bck_r_02, loc=<2.6804 0 0>, rot=<0 0 -0.0906 -0.9959>}
 tentacle_bck_r_04 {parent=tentacle_bck_r_03, loc=<2.0857 0 0>, rot=<0 0 -0.0471 -0.9989>}
 tentacle_bck_r_05 {parent=tentacle_bck_r_04, loc=<2.4973 0 0>, rot=<0 0 0 -1>}
 hood {parent=pack, loc=<0 -9.3654 5.4313>, rot=<0 -0.7071 0 -0.7071>}
]
variables [pawnSpeed {type=double, default=0}]
