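# Animation Blueprint for the Angler backpack, adding physics and procedural tentacle motion to its skeleton.
# Two RigidBody sims in pack-bone space share one physics asset, with a bool blend choosing weak gravity when DBNO and normal gravity otherwise.
# The result feeds a Control Rig driven by tentacleSpeed, remapped from 2D pawn speed each update, then a small additive offset shifts the pack bone.
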AnimBlueprint /BRCosmetics/Accessories/FORT_Backpacks/Backpack_F_MED_Angler/Meshes/F_MED_Angler_Pack_AnimBP
target_skeleton=&/F_MED_Angler_Pack_Skeleton
graphs [
 AnimGraph {nodes [
   n1 = RigidBody(ComponentPose=Local To Component(LocalPose=Local Space Ref Pose())) {
    bOverrideWorldGravity=true
    overridePhysicsAsset=&/F_MED_Angler_Pack_Physics
    simulationSpace=BaseBoneSpace
    alphaBoolBlend {blendOption=Linear}
    baseBoneRef {boneName=pack}
    componentLinearAccScale=<0.5 0.4 0.4>
    componentLinearVelScale=<0.4 0.4 0.4>
    overrideWorldGravity {z=-50}
   }
   n2 = RigidBody(ComponentPose=Local To Component(LocalPose=Local Space Ref Pose())) {
    bOverrideWorldGravity=true
    overridePhysicsAsset=&F_MED_Angler_Pack_Physics
    simulationSpace=BaseBoneSpace
    alphaBoolBlend {blendOption=Linear}
    baseBoneRef {boneName=pack}
    componentLinearAccScale=<0.5 0.4 0.4>
    componentLinearVelScale=<0.4 0.4 0.4>
    overrideWorldGravity {z=-980}
   }
   n3 = Control Rig!(
    Source=Blend Poses by bool(BlendPose_0=Component To Local(ComponentPose=@n1), BlendPose_1=Component To Local(ComponentPose=@n2), bActiveValue=Get bIsDBNO()) {blendType=Linear}
    Alpha
    pawnSpeed:double=Get tentacleSpeed() -> (tentacleSpeed:double)
   ) -> (Pose) {
    alphaBoolBlend {blendOption=Linear}
    controlRigAssetReference {blueprintRigClass=/BRCosmetics/Accessories/FORT_Backpacks/Backpack_F_MED_Angler/Meshes/F_MED_Angler_Pack_Skeleton_CtrlRig/F_MED_Angler_Pack_Skeleton_CtrlRig.F_MED_Angler_Pack_Skeleton_CtrlRig_C}
   }
   n4 = "Transform (Modify) Bone - Bone: pack"(ComponentPose=Local To Component(LocalPose=@n3), Translation="0,-2.5,0", Rotation="0,0,0") {translationMode=BMM_Additive, alphaBoolBlend {blendOption=Linear}, boneToModify {boneName=pack}, translation {y=-2.5}}
   Output Pose(Result=Component To Local(ComponentPose=@n4)) {layerGroup=None}
  ]}
 EventGraph {nodes [
   Event BlueprintUpdateAnimation()
   > Set tentacleSpeed(tentacleSpeed:double=MapRangeClamped(Value=Get PawnSpeed2D(), InRangeA=0, InRangeB=550, OutRangeA=1, OutRangeB=4)) -> (Output_Get:double)
   TryGetPawnOwner()
  ]}
]
variables [tentacleSpeed {type=number}]
